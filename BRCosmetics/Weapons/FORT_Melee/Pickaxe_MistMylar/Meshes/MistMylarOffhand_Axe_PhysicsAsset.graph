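# Physics asset for the MistMylar offhand axe pickaxe mesh, a melee cosmetic weapon.
# It defines a single capsule body on the b_AR_Root bone, offset and rotated relative to that bone.
# It has no constraints, so the axe simulates or collides as one rigid shape.
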PhysicsAsset /BRCosmetics/Weapons/FORT_Melee/Pickaxe_MistMylar/Meshes/MistMylarOffhand_Axe_PhysicsAsset
bodies {b_AR_Root [{shapeType=Capsule, length=25.8543, radius=10.9081, center=<1.4467 -1.9437 0.5303>, rotation=<rot -21.8564 -90 3.4074>}]}
constraints []
