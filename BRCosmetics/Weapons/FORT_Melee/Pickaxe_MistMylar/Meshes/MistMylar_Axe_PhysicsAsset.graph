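# Physics asset for the MistMylar axe, a cosmetic melee pickaxe weapon.
# It defines a single body on the b_AR_Root bone, shaped as one offset, rotated capsule.
# It has no constraints, so the asset holds no jointed or secondary physics chains.
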PhysicsAsset /BRCosmetics/Weapons/FORT_Melee/Pickaxe_MistMylar/Meshes/MistMylar_Axe_PhysicsAsset
bodies {b_AR_Root [{shapeType=Capsule, length=25.8543, radius=10.8665, center=<-1.9272 1.418 0.5388>, rotation=<rot 3.6706 -90 -21.8158>}]}
constraints []
